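# Parent masked subsurface material for the GrainSpike quadruped companion's shell-fur body, with an optional glow style.
# Character-parent functions bracket a fur chain covering custom color, adjustments, painting, shells and WPO shrink.
# GrainSpike logic adds two-color emissive shaped by downward-facing normals, a detail mask and time of day, and tints upward-facing base color.
# UseGlowStyle blends this over the standard fur result; a normalized shell-layer value provides a root-to-tip mask.
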Material /CosmeticCompanions/Assets/Quadruped/GrainSpike/Materials/Fur/M_GrainSpike_Body_FurParent
graph {
 properties {
  bUseMaterialAttributes=true
  bUsedWithMeshDeformer=true
  bUsedWithSkeletalMesh=true
  blendMode=BLEND_Masked
  pixelDepthOffsetMode=PDOM_Legacy
  shadingModel=MSM_Subsurface
  usedShadingModels=Subsurface
 }
 nodes [
  n1 = MaterialFunctionCall() {materialFunction=&MF_CharacterParent_2023_PreFX}
  MaterialFunctionCall(Material=NamedRerouteUsage(), M=@n1.M, Pre Skinned Local Position=@n1.M, Distance Blend=@n1.M, Fresnel=@n1.M) {materialFunction=&MF_CharacterParent_2023_PostFX}
  NamedRerouteDeclaration(None=@n1.Tech Art Mask) {name=TechArtMask, nodeColor=<rgba 0.2188 0 1 1>}
  NamedRerouteDeclaration(None=@n1.Result) {name=Material, nodeColor=<rgba 0 1 0.1094 1>}
  n2 = NamedRerouteUsage()
  n3 = NamedRerouteUsage()
  n4 = Add(
   A=OneMinus(None=ComponentMask(None=@n3) {r=true})
   B=ScalarParameter() {defaultValue=0.1, group=GrainSpike, parameterName=Global_Emissive_intensity}
  )
  n5 = LinearInterpolate(Alpha=NamedRerouteUsage()) {constA=1, constB=0.1}
  n6 = VectorParameter() {group=GrainSpike, parameterName=EmissiveColorOffset, defaultValue=<rgba 0.08 3.6264 4 1>}
  n7 = VectorParameter() {group=GrainSpike, parameterName=EmissiveColorOffset2, defaultValue=<rgba 0 2.246 3.5 1>}
  n8 = Multiply(
   A=Multiply(A=ComponentMask(None=@n2) {r=true}, B=Multiply(A=@n4, B=@n5))
   B=Multiply(A=NamedRerouteUsage(), B=LinearInterpolate(A=@n6.RGB, B=@n7.RGB))
  )
  n9 = Constant() {r=4}
  n10 = DotProduct(A=VertexNormalWS(), B=Constant4Vector() {constant=<rgba 0 0 -1 1>}.RGB)
  n11 = Constant() {r=0.75}
  n12 = Saturate(None=Multiply(A=Power(Base=@n10, Exp=@n11) {constExponent=1}, B=ScalarParameter() {defaultValue=1.05, group=GrainSpike, parameterName=Bottom brightness}))
  n13 = Multiply(
   A=TextureSample() {samplerType=SAMPLERTYPE_Masks, texture=/CosmeticCompanions/Assets/Quadruped/GrainSpike/Textures/T_GrainSpike_Body_TA_Mask}.R
   B=ScalarParameter() {defaultValue=5, group=GrainSpike, parameterName="DetailE,missiveBrightness"}
  ) {constB=3}
  n14 = LinearInterpolate(
   A=ScalarParameter() {defaultValue=0.5, group=GrainSpike, parameterName=Top brightness}
   B=MaterialFunctionCall(Value_Morning=@n9, Value_Day=Constant() {r=5}, Value_Evening=@n9, Value_Night=Constant() {r=4}) {materialFunction=&Scalar_Time_of_Day}.Result
   Alpha=NamedRerouteDeclaration(None=Max(A=@n12, B=@n13)) {name=TopBottomMask, nodeColor=<rgba 0.9219 1 0 1>}
  )
  NamedRerouteDeclaration(None=Add(A=NamedRerouteUsage(), B=Multiply(A=@n8, B=@n14))) {name=OutPutEmissive, nodeColor=<rgba 0.2656 1 0 1>}
  NamedRerouteDeclaration(None=@n1.M) {name=Mask, nodeColor=<rgba 1 0 0.2109 1>}
  n15 = MaterialFunctionCall!(Material=NamedRerouteUsage(), TechArtMask=NamedRerouteUsage()) {materialFunction=/Game/Global/Materials/Character_Parent_2023/Fur_Parent_2023/MF_Fur_CustomColor}
  n16 = NamedRerouteUsage()
  n17 = MaterialFunctionCall!(
   In=MaterialFunctionCall!(
    In=MaterialFunctionCall!(MA=@n15.Result, TechArtMask=@n16) {materialFunction=/Game/Global/Materials/Character_Parent_2023/Fur_Parent_2023/MF_Fur_ParamAdjustments}.Result
    TechArtMask=NamedRerouteUsage()
    Mask Input=NamedRerouteUsage()
   ) {materialFunction=/Game/Global/Materials/Character_Parent_2023/Fur_Parent_2023/MF_Fur_Painted}.Result
   TechArtMask=NamedRerouteUsage()
  ) {materialFunction=/Game/Global/Materials/Character_Parent_2023/Fur_Parent_2023/MF_Fur_Shells}
  n18 = GetMaterialAttributes(None=MaterialFunctionCall!(Material=@n17.Result, Mask=NamedRerouteUsage(), Fur Length=@n17.Result) {materialFunction=/Game/Global/Materials/Character_Parent_2023/Fur_Parent_2023/MF_FurWPOShrink}.Result) {attributeGetTypes [
    #1
    #2
    #3
    #4
   ]}
  n19 = Constant() {r=1}
  n20 = SetMaterialAttributes!(MaterialAttributes=@n18.MaterialAttributes, Base Color, Subsurface Color, Opacity=@n19, Emissive Color) {attributeSetTypes [
    #1
    #3
    #2
    #4
   ]}
  n21 = ScalarParameter() {defaultValue=0.35, group=GrainSpike, parameterName=FreselTop brightness, sliderMax=1}
  n22 = VertexNormalWS()
  n23 = Constant4Vector() {constant=<rgba 0 0 1 1>}
  n24 = Saturate(None=Power(Base=DotProduct(A=@n22, B=@n23.RGB), Exp=Constant() {r=2}) {constExponent=1.5})
  n25 = SetMaterialAttributes(
   MaterialAttributes=@n18.MaterialAttributes
   Base Color=Multiply(
    A=NamedRerouteDeclaration(None=@n18.Base Color) {name=InputBaseColor, nodeColor=<rgba 0 0.4609 1 1>}
    B=LinearInterpolate(B=@n21, Alpha=@n24) {constA=1}
   )
   Opacity=@n19
   Emissive Color=NamedRerouteUsage()
  ) {attributeSetTypes [
    #1
    #3
    #2
    #4
   ]}
  n26 = ScalarParameter() {group=GrainSpike, parameterName=UseGlowStyle, sortPriority=1}
  NamedRerouteDeclaration(None=BlendMaterialAttributes(A=@n20, B=@n25, Alpha=NamedRerouteDeclaration(None=@n26) {name=UseGlowStyle, nodeColor=<rgba 0 0.6016 1 1>})) {name=OUT_Material, nodeColor=<rgba 0 0.5312 1 1>}
  n27 = ShellMeshMaterialExpressionNormalizedShellLayer() {is=class ShellMeshMaterialExpressionNormalizedShellLayer}
  n28 = VertexInterpolator(VS=@n27)
  NamedRerouteDeclaration(None=NamedRerouteDeclaration(None=ShaderStageSwitch(PixelShader=@n28.PS, VertexShader=@n27)) {name=NormalizedShellLayer, nodeColor=<rgba 0.2656 0 1 1>}) {name=FurRootTipMask, nodeColor=<rgba 0 1 0.8594 1>}
  NamedRerouteDeclaration(None=@n18.Emissive Color) {name=InputEmissiveColor, nodeColor=<rgba 1 0.5391 0 1>}
 ]
}
